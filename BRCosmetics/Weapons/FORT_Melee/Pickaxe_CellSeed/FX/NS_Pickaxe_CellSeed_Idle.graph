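# Idle Niagara effect for the CellSeed pickaxe, adding a blue glow, lights and sparkles.
# NE_Glow bursts ten persistent sprites, using partition index to pick socket, size, alignment and color, with a front-end preview color and alpha variant.
# NE_Light spawns two vertically offset lights once, gated by loop count.
# NE_Sparkles continuously spawns short-lived socket-sampled twinkles with two sprite renderers and derived larger, dimmer attributes; time-of-day scalars adjust brightness.
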NiagaraSystem /BRCosmetics/Weapons/FORT_Melee/Pickaxe_CellSeed/FX/NS_Pickaxe_CellSeed_Idle
dependencies {
 usedDataInterfaces [
  &NiagaraDataInterfaceArrayInt32
  &NiagaraDataInterfaceArrayFloat3
  &NiagaraDataInterfaceCurve
  &NiagaraDataInterfaceSocketReader
 ]
 usedDynamicInputs [&RandomRangeVector]
 usedModules [
  &self:SystemSpawnScript.SystemScriptSource.SystemScriptGraph.NiagaraNodeAssignment_1.SetVariables_#1
  &SystemState
 ]
}
emitters {
 NE_Glow {
  emitterUpdateScript [
   &EmitterState(
    Inactive Response=::NewEnumerator1
    Life Cycle Mode=::NewEnumerator1
    Loop Behavior=::NewEnumerator1
    Loop Duration Mode=::NewEnumerator1
    UseLoopDelay=false
   )
   &SpawnBurst_Instantaneous(Spawn Count=10, Use Spawn Probability=false, UseLoopCountLimit=false)
  ]
  particleSpawnScript [
   &PartitionParticles(Number Of Partitions=10)
   &InitializeParticle(
    Color=&MakeCustomLinearColorFromBool(
     Bool=$User.bIsFrontEndPreview
     True Color=&MergeHelper_ColorSwitch(
      0=<rgba 0 1 2 0.25>
      1=<rgba 0 0.75 1.5 0.25>
      2=<rgba 0 0.75 1.5 0.25>
      3=<rgba 0 1 2 0.25>
      4=<rgba 0 1 2 0.25>
      5=<rgba 0 1 2 0.25>
      6=<rgba 0 1 2 0.25>
      7=<rgba 0 1 2 1>
      8=<rgba 0 0.75 1.5 1>
      9=<rgba 0 1 2 0.5>
      Index=$Particles.PartitionParticles.Partition
      Num=::NewEnumerator8
     )
     False Color=&MergeHelper_ColorSwitch(
      0=<rgba 0.7 1.05 1.4 0.1>
      1=<rgba 0.575 0.8625 1.15 0.1>
      2=<rgba 0.6 1.05 1.5 0.1>
      3=<rgba 0.48 0.84 1.2 0.2>
      4=<rgba 0.6 1.05 1.5 0.1>
      5=<rgba 0.6 1.05 1.5 0.1>
      6=<rgba 0.6 1.2 1.5 0.1>
      7=<rgba 0.6 1.2 1.5 0.25>
      8=<rgba 0.6 1.2 1.5 0.25>
      9=<rgba 0.8 1.6 2 0.15>
      Index=$Particles.PartitionParticles.Partition
      Num=::NewEnumerator8
     )
    )
    Sprite Size=&MergeHelper_Vector2Switch(
     0=<55 110>
     1=<55 110>
     2=<55 110>
     3=<50 80>
     4=<30 55>
     5=<30 55>
     6=<25 50>
     7=<25 40>
     8=<15 33>
     9=<20 30>
     Index=$Particles.PartitionParticles.Partition
     Num=::NewEnumerator8
    )
    Sprite Size Mode=::NewEnumerator4
    UsePositionOffset=false
    Write Material Random=true
   )
   &CameraOffset(Camera Offset Amount=3)
  ]
  particleUpdateScript [
   &ParticleState(Kill Particles When Lifetime Has Elapsed=false, Let Infinitely Lived Particles Die When Emitter Deactivates=true)
   &SocketLocation(
    SamplePosOffset=true
    Sampled Position Offset=&SelectVectorFromArray(
     Direct Array Index=$Particles.PartitionParticles.Partition
     Vector Selection Array=DataInterface() {propertyValues {FloatData [<0 0 0>, <1.5 0 0>, <0 0 0>, <0 0 0>, <0 3 0>, <0 -3 0>, <0 0 0>, <0 0 0>, <0 0 0>, <0 0 0>]}}
    )
    Socket Index=&SelectIntFromArray(
     Direct Array Index=$Particles.PartitionParticles.Partition
     Int Selection Array=DataInterface() {propertyValues {IntData [0,1,2,3,4,4,5,6,7,8]}}
    )
    Socket Reader=$System.Socket Reader
    Socket Sampling Mode=::NewEnumerator3
    VelocitySpeedLimit=false
   )
   &SpriteFacingAndAlignment(
    Sprite Alignment=&SelectVectorFromArray(
     Direct Array Index=$Particles.PartitionParticles.Partition
     Vector Selection Array=DataInterface() {propertyValues {FloatData [<-0.2 0 1>, <-0.15 0 1>, <0.05 0 1>, <0 0 1>, <1 -0.5 1>, <1 0.5 1>, <0.3 0 1>, <0 0 1>, <0.5 0 1>, <1 0 0>]}}
    )
    WriteAlignment=true
    WriteFacing=false
   )
   &ScaleColor(Color Scale:LinearColor=<rgba 1 1 1 0.5>, Scale Mode=3)
   &ScaleColor(
    Scale Alpha=&MakeCustomFloatFromBool(
     Bool=$User.bIsFrontEndPreview
     True Float=0.15
     False Float=&FNTimeOfDayScalar(Night Scale=0.2, Morning Scale=0.15, Day Scale=0.35, Evening Scale=0.15)
    )
    ScaleA=true
    ScaleRGB=true
   ) {module=ScaleColor001}
  ]
  rendererClasses [&NiagaraSpriteRendererProperties]
  renderers [{rendererClass=&NiagaraSpriteRendererProperties}]
 }
 NE_Light {
  emitterUpdateScript [
   &EmitterState(Life Cycle Mode=::NewEnumerator1, Loop Delay=1, Loop Duration=0.1, UseLoopDelay=false)
   &SpawnBurst_Instantaneous(Spawn Probability=HlslExpression() {hlslExpression="Emitter.LoopCount == 0"}, Use Spawn Probability=true, UseLoopCountLimit=false)
   &SpawnBurst_Instantaneous(
    Spawn Group=1
    Spawn Probability=HlslExpression() {hlslExpression="Emitter.LoopCount == 0"}
    Use Spawn Probability=true
    UseLoopCountLimit=false
   ) {module=SpawnBurst_Instantaneous001}
  ]
  particleSpawnScript [&InitializeParticle(Color Mode=::NewEnumerator0, UsePositionOffset=false, Write Material Random=true)]
  particleUpdateScript [
   &ParticleState(Kill Particles When Lifetime Has Elapsed=false, Let Infinitely Lived Particles Die When Emitter Deactivates=true)
   &FN_Light_Attributes(
    Color=&FNTimeOfDayColor(Night Color=<rgba 0.3 0.65 1 1>, Morning Color=<rgba 0.3 0.65 1 1>, Day Color=<rgba 0.3 0.65 1 1>, Evening Color=<rgba 0.3 0.65 1 1>)
    Create Color Attribute=true
    Create Exponent Attribute=true
    Create Position Attribute=true
    Create Radius Attribute=true
    Exponent=3
    Position=&AddVectorToPosition(
     Position=$Particles.Position
     Vector=&MergeHelper_VectorSwitch(0=<0 0 50>, 1=<0 0 -50>, 2=?, 3=?, Index=&SpawnGroup(), Num=::NewEnumerator0)
    )
    Radius=40
    Scale Color=25
   )
  ]
  rendererClasses [&NiagaraLightRendererProperties]
  renderers [{rendererClass=&NiagaraLightRendererProperties, rendererIndex=0}]
 }
 NE_Sparkles {
  emitterUpdateScript [
   &EmitterState(Inactive Response=::NewEnumerator1, Life Cycle Mode=::NewEnumerator1, UseLoopDelay=false)
   &SpawnRate(
    Spawn Probability=&MakeCustomFloatFromBool(Bool=$System.bIsRunning, True Float=1, False Float=0.5)
    SpawnRate=5
    Use Spawn Probability=true
   )
  ]
  particleSpawnScript [
   &FN_SetSpawnGroup(Number of Spawn Groups=2)
   &InitializeParticle(
    Color=<rgba 0.5 0.8333 1 1>
    Lifetime Max=0.5
    Lifetime Min=0.3
    Lifetime Mode=::NewEnumerator1
    Position Offset=&MergeHelper_VectorSwitch(0=<-15 -10 -15>, 1=<15 -10 -15>, 2=?, 3=?, Index=$Particles.SpawnGroup, Num=::NewEnumerator0)
    Sprite Rotation Mode=::NewEnumerator2
    Sprite Size Mode=::NewEnumerator1
    Uniform Sprite Size Min=3
    UsePositionOffset=true
    Write Material Random=true
   )
   &SocketLocation(
    OverrideRandomSeed=false
    SamplePosOffset=false
    Socket Reader=$System.Socket Reader
    Socket Sampling Mode=::NewEnumerator1
    VelocitySpeedLimit=false
   )
  ]
  particleUpdateScript [
   &ParticleState()
   &SpriteRotationRate(Rotation Rate=&FloatFromCurve(
     FloatCurve=DataInterface() {propertyValues {Curve {
        defaultValue=1
        keys [
         {arriveTangent=-0.1154, interpMode=RCIM_Cubic, leaveTangent=-0.1154, tangentMode=RCTM_User, time=0, value=1}
         {arriveTangent=-2.2067, interpMode=RCIM_Cubic, leaveTangent=-2.2067, tangentMode=RCTM_User, time=1, value=0}
        ]
       }}}
     Scale Curve=2
    ))
   &ScaleSpriteSize(Uniform Curve Sprite Scale=DataInterface() {propertyValues {Curve {
       defaultValue=0
       keys [
        <at 0 0 RCIM_Cubic>
        {arriveTangent=-2.5249, interpMode=RCIM_Cubic, leaveTangent=-2.5249, tangentMode=RCTM_User, time=0.1, value=1}
        {arriveTangent=-0.0405, interpMode=RCIM_Cubic, leaveTangent=-0.0405, tangentMode=RCTM_User, time=1, value=0}
       ]
      }}})
   &ScaleColor(
    Scale Alpha=&Multiply_Float(
     A=1
     B=&FloatFromCurve(FloatCurve=DataInterface() {propertyValues {Curve {keys [
          {arriveTangent=-2.5587, interpMode=RCIM_Cubic, leaveTangent=-2.5587, tangentMode=RCTM_User, time=0, value=1}
          {interpMode=RCIM_Cubic, tangentMode=RCTM_User, time=1, value=0}
         ]}}})
    )
    Scale RGB=&Multiply_VectorByFloat(
     Vector=<1 1 1>
     Float=&Multiply_Float(
      A=&RandomRangeFloat(Minimum=0.3, Maximum=1, Evaluation Type=::NewEnumerator1)
      B=&FNTimeOfDayScalar(Night Scale=2.5, Morning Scale=5, Day Scale=5, Evening Scale=5)
     )
    )
    ScaleA=true
    ScaleRGB=true
   )
   &DynamicMaterialParameters(
    Index 0 Param 1=&FloatFromCurve(FloatCurve=DataInterface() {propertyValues {Curve {keys [{tangentMode=RCTM_User, time=0, value=0}, {tangentMode=RCTM_User, time=1, value=1}]}}})
    Index 0 Param 2=0.7
    Index 0 Param 3=&FloatFromCurve(FloatCurve=DataInterface() {propertyValues {Curve {keys [<at 0 0 RCIM_Cubic>, {arriveTangent=0.2057, interpMode=RCIM_Cubic, leaveTangent=0.2057, tangentMode=RCTM_User, time=1, value=3}]}}})
    Index 0 Param 4=1
    Index 1 Param 1=0.5
    Index 1 Param 3=&Lerp_Float(A=1, B=2, Alpha=$Particles.Random)
    Index 1 Param 4=1
    Module.Index0Float0Write=true
    Module.Index0Float1Write=true
    Module.Index0Float2Write=true
    Module.Index0Float3Write=true
    Module.Index1Float0Write=true
    Module.Index1Float1Write=true
    Module.Index1Float2Write=true
    Module.Index1Float3Write=true
    Write Parameter Index 1=::NewEnumerator1
   )
   &SolveForcesAndVelocity(Clamp Velocity=false, Limit Acceleration=false)
   &CameraOffset(Camera Offset Amount=5)
   /SetVariables(
    Particles.ColorBig:LinearColor=&MultiplyLinearColor(Linear Color=$Particles.Color, Scale Factor=<0.5 0.5 0.5 0.5>)
    Particles.SizeBig:Vector2f=&Multiply_Vector2D_ByFloat(Float=5, Vector2D=$Particles.SpriteSize)
   ) {isSetParameters=true}
  ]
  rendererClasses [&NiagaraSpriteRendererProperties]
  renderers [
   {rendererClass=&NiagaraSpriteRendererProperties}
   {rendererClass=&NiagaraSpriteRendererProperties, rendererIndex=1}
  ]
 }
}
summary {
 systemName=NS_Pickaxe_CellSeed_Idle
 emitters [
  {emitterName=NE_Glow, rendererClasses [&NiagaraSpriteRendererProperties]}
  {emitterName=NE_Light, rendererClasses [&NiagaraLightRendererProperties]}
  {emitterName=NE_Sparkles, rendererClasses [&NiagaraSpriteRendererProperties]}
 ]
 userVariables [
  User.bIsFrontEnd {type=NiagaraBool}
  User.bIsFrontEndPreview {type=NiagaraBool}
  User.bisSwinging {type=NiagaraBool}
  User.PhysicalSurface {type=NiagaraInt32}
  User.RibbonIDOffset {type=NiagaraInt32}
  User.SwingSectionIndex {type=NiagaraInt32}
  User.VariantID {type=NiagaraFloat, defaultValue=0, description=""}
 ]
}
