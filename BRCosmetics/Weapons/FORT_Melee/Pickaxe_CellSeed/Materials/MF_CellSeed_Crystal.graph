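# Material function for the CellSeed pickaxe that layers a faux-refractive crystal look onto input material attributes, blended by Alpha.
# A depth texture drives parallax UV offsets for fake interior detail, tinted and adjusted by time of day.
# Interior normals drive sun-direction backlighting, alongside refracted skylight sampling, Fresnel, edge colors and an optional emissive tip sphere mask.
# Named reroutes link the stages, and several experimental branches are left unconnected.
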MaterialFunction /BRCosmetics/Weapons/FORT_Melee/Pickaxe_CellSeed/Materials/MF_CellSeed_Crystal
graph {nodes [
  n1 = Add(A=Constant(), B=Constant3Vector() {constant=<rgba 0 7 0 0>})
  n2 = MaterialFunctionCall!(
   Origin=TransformPosition(None=@n1) {transformType=TRANSFORMPOSSOURCE_World}
   Radius=Constant() {r=11}
   Z Axis
   XY Scale
   X Axis
   Capped?=StaticBool()
   Z Min-Max
  ) {materialFunction=/Engine/Functions/Engine_MaterialFunctions01/Math/CylinderIntersection}
  n3 = ComponentMask(None=@n2.Ray Exit Position) {r=true}
  Abs(None=@n3)
  n4 = TextureCoordinate()
  n5 = Multiply(
   A=NamedRerouteDeclaration(None=@n4) {name=UVCoord, nodeColor=<rgba 0.5937 1 0 1>}
   B=ScalarParameter() {defaultValue=4, group=Crystal, parameterName=InteriorUVTiling, sortPriority=4}
  )
  n6 = TextureSampleParameter2D(UVs=Multiply(A=@n4, B=ScalarParameter() {defaultValue=4, group=Crystal, parameterName=DepthUVTiling, sortPriority=8})) {
   group=Crystal
   parameterName=DepthTexture
   samplerType=SAMPLERTYPE_LinearGrayscale
   sortPriority=9
   texture=/Game/Global/Textures/Default/Masks_and_Gradients/T_MF_Crystal2
  }
  n7 = LinearInterpolate(
   A=ScalarParameter() {group=Crystal, parameterName=DepthMinimum, sortPriority=6}
   B=ScalarParameter() {defaultValue=3, group=Crystal, parameterName=DepthMaximum, sortPriority=7}
   Alpha=NamedRerouteDeclaration(None=@n6.R) {name=SparksDepth, nodeColor=<rgba 0.1156 1 0.3086 1>}
  ) {constA=-4, constB=-20}
  BumpOffset(Coordinate=@n5, Height=@n7)
  Constant3Vector() {constant=<rgba 0 0 1 0>}
  n8 = Multiply(A=@n2.Ray Exit Position) {constB=0.1}
  n9 = ComponentMask(None=@n8) {r=true}
  n10 = ComponentMask(None=@n8) {g=true}
  n11 = Arctangent2Fast(Y=@n9, X=@n10)
  ConstantBiasScale(None=@n11)
  Frac(None=Divide(A=@n11) {constB=6.2832})
  n12 = VectorParameter() {group=Crystal, parameterName=BaseColorMultiply, sortPriority=0, defaultValue=<rgba 0.0277 0.0444 0.0781 1>}
  n13 = TextureSampleParameter2D(UVs=NamedRerouteUsage()) {group=Crystal, parameterName=InteriorTexture, sortPriority=5, texture=/Game/Global/Textures/Default/Masks_and_Gradients/T_MF_Crystal}
  n14 = ScalarParameter() {defaultValue=0.1, group=Crystal, parameterName=InteriorPrismColors_Saturation, sortPriority=2}
  n15 = MaterialFunctionCall!(
   In=Desaturation(None=@n13.RGB, Fraction=OneMinus(None=@n14))
   Contrast=ScalarParameter() {group=Crystal, parameterName=InteriorPrismColors_Contrast, sortPriority=2}
  ) {materialFunction=/Engine/Functions/Engine_MaterialFunctions01/ImageAdjustment/CheapContrast_RGB}
  n16 = LinearInterpolate(
   A=Multiply(A=@n12.RGB, B=NamedRerouteUsage())
   B=Multiply(A=Multiply(A=@n12.RGB, B=@n15.Result), B=ScalarParameter() {defaultValue=12, group=Crystal, parameterName=InteriorIntensity, sortPriority=1}) {constB=0.1}
   Alpha=ScalarParameter() {defaultValue=1, group=Crystal, parameterName=InteriorContrast, sortPriority=3}
  )
  n17 = Power(Base=@n16, Exp=Constant() {r=0.25}) {constExponent=0.5}
  n18 = Constant() {r=0.3}
  n19 = LinearInterpolate(
   A=@n16
   B=Multiply(A=NamedRerouteUsage(), B=Desaturation(None=@n17, Fraction=@n18))
   Alpha=ScalarParameter() {defaultValue=0.4, parameterName=RefractionAmount}
  )
  n20 = Multiply(A=@n19, B=ScalarParameter() {defaultValue=1.2, parameterName=Interior_TOD_Night_Intensity})
  n21 = NamedRerouteDeclaration(None=MaterialFunctionCall(Color_Morning=@n19, Color_Day=@n19, Color_Evening=@n20, Color_Night=@n20) {materialFunction=&Color_Time_of_Day}.Result) {name=InteriorColor, nodeColor=<rgba 1 0 0.4687 1>}
  n22 = VectorParameter() {group=Crystal, parameterName=FresnelColor, sortPriority=19, defaultValue=<rgba 0.6027 0.898 1 1>}
  n23 = NamedRerouteUsage()
  n24 = Desaturation(None=NamedRerouteUsage(), Fraction=Constant() {r=1})
  n25 = Power(Base=@n24, Exp=ScalarParameter() {defaultValue=1.25, group=Crystal, parameterName=Edges_Exp}) {constExponent=1.5}
  n26 = NamedRerouteUsage()
  n27 = NamedRerouteUsage()
  n28 = Add(
   A=Power(Base=@n24, Exp=ScalarParameter() {defaultValue=1.5, group=Crystal, parameterName=Edges_Emissive_Exp})
   B=Multiply(A=Multiply(A=@n24, B=@n27), B=ScalarParameter() {defaultValue=10, parameterName=B})
  )
  n29 = VectorParameter() {group=Crystal, parameterName=Edges_Emissive_Color, defaultValue=<rgba 0.3937 0.5485 1 1>}
  n30 = SetMaterialAttributes!(
   MaterialAttributes=NamedRerouteUsage()
   BaseColor=LinearInterpolate(
    A=LinearInterpolate(A=@n21, B=@n22.RGB, Alpha=@n23)
    B=VectorParameter() {group=Crystal, parameterName=Edges_Color, defaultValue=<rgba 1 1 1 1>}.RGB
    Alpha=@n25
   )
   EmissiveColor=Add(A=LinearInterpolate(A=@n26, Alpha=@n25) {constB=0}, B=Multiply(A=@n28, B=@n29.RGB))
   Metallic=Constant()
   Roughness=Add(A=NamedRerouteUsage(), B=ScalarParameter() {defaultValue=-0.1, group=Crystal, parameterName=Roughness_Bias})
  ) {attributeSetTypes [
    #1
    #2
    #3
    #4
   ]}
  n31 = FunctionInput() {bUsePreviewValueAsDefault=true, inputName=Alpha, inputType=FunctionInput_Scalar, sortPriority=1, previewValue=<1 1 1 1>}
  FunctionOutput(None=BlendMaterialAttributes(A=NamedRerouteUsage(), B=@n30, Alpha=@n31))
  MaterialFunctionCall() {materialFunction=&VectorToRadialValue}
  MaterialFunctionCall() {materialFunction=&ObjectPivotPoint}
  MaterialFunctionCall() {materialFunction=&CameraDirectionVector}
  Max(A=@n9, B=@n10)
  n32 = PixelNormalWS()
  n33 = CameraVectorWS()
  n34 = OneMinus(None=Saturate(None=DotProduct(A=@n32, B=@n33)))
  Multiply(A=@n34, B=@n34)
  n35 = Constant3Vector() {constant=<rgba 1 0 0 0>}
  n36 = Add(A=MaterialFunctionCall(
    Plane Normal Axis=CameraVectorWS()
    Plane X Axis=Transform(None=@n35) {transformSourceType=TRANSFORMSOURCE_View}
    UV Scale=Constant() {r=100}
   ) {materialFunction=&VirtualPlaneCoordinates}.UVs) {constB=-0.5}
  Multiply(A=@n36)
  n37 = GetMaterialAttributes(None=FunctionInput() {inputName=Material, inputType=FunctionInput_MaterialAttributes}) {attributeGetTypes [#1, #2, #4]}
  NamedRerouteDeclaration(None=@n37.EmissiveColor) {name=Emissive, nodeColor=<rgba 1 0 0.75 1>}
  NamedRerouteDeclaration(None=@n37.MaterialAttributes) {name=MA, nodeColor=<rgba 0.3406 1 0.0156 1>}
  n38 = Constant3Vector() {constant=<rgba 0 0 1 0>}
  n39 = TextureSampleParameter2D(UVs=NamedRerouteUsage()) {
   group=Crystal
   parameterName=BacklightNormal
   samplerType=SAMPLERTYPE_Normal
   sortPriority=18
   texture=/Game/Global/Textures/Default/Masks_and_Gradients/T_MF_Crystal3
  }
  n40 = ScalarParameter() {defaultValue=0.2, group=Crystal, parameterName=BacklightNormalIntensity, sortPriority=15}
  n41 = NamedRerouteDeclaration(None=Transform(None=LinearInterpolate(A=@n38, B=@n39.RGB, Alpha=@n40) {constAlpha=0.1})) {name=InteriorNormals, nodeColor=<rgba 0.5469 0 1 1>}
  n42 = CameraVectorWS()
  n43 = Add(A=@n41, B=Multiply(A=ScalarParameter() {defaultValue=1, group=Crystal, parameterName=BacklightViewAngle, sortPriority=16}, B=@n42))
  n44 = SkyAtmosphereLightDirection()
  n45 = Multiply(A=DotProduct(A=Normalize(VectorInput=@n43), B=Multiply(A=@n44) {constB=-1})) {constB=0.5}
  n46 = Power(
   Base=Saturate(None=Add(A=@n45) {constB=0.5})
   Exp=ScalarParameter() {defaultValue=8, group=Crystal, parameterName=BacklightAperture, sortPriority=17}
  )
  n47 = Power(Base=Saturate(None=DotProduct(A=@n41, B=@n42)), Exp=ScalarParameter() {defaultValue=0.5, parameterName=Exp})
  n48 = ScalarParameter() {defaultValue=2, group=Crystal, parameterName=BackLightIntensity_Morning, sortPriority=14}
  n49 = ScalarParameter() {defaultValue=3, group=Crystal, parameterName=BackLightIntensity_Day, sortPriority=14}
  n50 = ScalarParameter() {defaultValue=3, group=Crystal, parameterName=BackLightIntensity_Night, sortPriority=14}
  n51 = Max(
   A=Multiply(
    A=Multiply(A=@n46, B=@n47)
    B=MaterialFunctionCall(Value_Morning=@n48, Value_Day=@n49, Value_Evening=@n48, Value_Night=@n50) {materialFunction=&Scalar_Time_of_Day}.Result
   )
   B=ScalarParameter() {defaultValue=0.3, group=Crystal, parameterName=BacklightMin}
  )
  n52 = NamedRerouteUsage()
  n53 = VectorParameter() {group=Crystal, parameterName=BacklightColorMultiply, sortPriority=13, defaultValue=<rgba 0.1927 0.5078 1 1>}
  NamedRerouteDeclaration(None=Multiply(A=Multiply(A=@n51, B=@n52), B=@n53.RGB)) {name=Backlight, nodeColor=<rgba 1 0 0.2578 1>}
  NamedRerouteDeclaration(None=@n37.Roughness) {name=Roughness, nodeColor=<rgba 0.2422 0 1 1>}
  n54 = LinearInterpolate(A=@n9, B=@n10, Alpha=Saturate(None=@n3))
  n55 = Multiply(A=ComponentMask(None=@n8) {b=true})
  n56 = TextureSampleParameter2D(UVs=Multiply(A=AppendVector(A=@n54, B=@n55), B=ScalarParameter() {defaultValue=0.3, parameterName=B_1})) {
   group=Crystal
   parameterName=BacklightNormal
   samplerType=SAMPLERTYPE_Normal
   sortPriority=18
   texture=&T_MF_Crystal3
  }
  NamedRerouteDeclaration(None=Transform(None=@n56.RGB)) {name=CylinderNormals, nodeColor=<rgba 0 1 0.8594 1>}
  n57 = VertexInterpolator(VS=Add(A=LocalPosition() {localOrigin=InstancePreSkinning}.XYZ, B=Constant3Vector() {constant=<rgba 0 -7 -105 0>}))
  n58 = Constant()
  n59 = ScalarParameter() {defaultValue=30, parameterName=MaskTipRadius}
  NamedRerouteDeclaration(None=Multiply(
    A=SphereMask(A=@n57.PS, B=@n58, Radius=@n59) {attenuationRadius=30, hardnessPercent=0}
    B=ScalarParameter() {parameterName=UseSphereMaskForEmissiveTips}
   )) {name=TipMask, nodeColor=<rgba 1 0 0.4922 1>}
  NamedRerouteDeclaration(None=@n37.BaseColor) {name=BaseColor, nodeColor=<rgba 0.1661 0.3253 1 1>}
  n60 = Multiply(A=CameraVectorWS()) {constB=-1}
  n61 = NamedRerouteUsage()
  n62 = Normalize(VectorInput=Add(A=PixelNormalWS(), B=@n61))
  n63 = ScalarParameter() {defaultValue=1.33, group=Crystal, parameterName=Environment_IOR, sortPriority=22}
  n64 = VectorParameter() {group=Crystal, parameterName=ReflectionColor, sortPriority=21, defaultValue=<rgba 0.6771 0.8607 1 1>}
  n65 = Multiply(
   A=SkyLightEnvMapSample(
    Direction=MaterialFunctionCall!(Ray Direction=@n60, Surface Normal=@n62, Refractive Index Origin, Refractive Index Target=@n63) {materialFunction=/Engine/Functions/Engine_MaterialFunctions01/Vectors/Refract}.Vector
    Roughness=ScalarParameter() {group=Crystal, parameterName=Environment_Roughness, sortPriority=22}
   )
   B=@n64.RGB
  )
  n66 = CameraVectorWS()
  n67 = Power(Base=ConstantBiasScale(None=DotProduct(A=@n61, B=@n66)), Exp=ScalarParameter() {defaultValue=5, parameterName=RefractionFesnelExp}) {constExponent=8}
  NamedRerouteDeclaration(None=Multiply(A=Multiply(A=@n65, B=@n67), B=NamedRerouteUsage())) {name=Refraction, nodeColor=<rgba 1 0.1406 0 1>}
  NamedRerouteDeclaration(None=MaterialFunctionCall!(UV Coord=@n5, Height=@n7, ReferencePlane, HeightRatio, Normal) {materialFunction=/Game/Global/MaterialFunctions/MF_BumpOffset_Accurate}.Result) {name=UVOffsetCoord, nodeColor=<rgba 0.9453 0 1 1>}
  n68 = ScalarParameter() {defaultValue=1, group=Crystal, parameterName=FresnelExponent}
  n69 = Fresnel(ExponentIn=@n68) {baseReflectFraction=0, exponent=1}
  n70 = Constant() {r=1}
  n71 = ScalarParameter() {defaultValue=0.5, parameterName=FresnelRadius}
  NamedRerouteDeclaration(None=Multiply(
    A=SphereMask(A=@n69, B=@n70, Radius=@n71) {attenuationRadius=1, hardnessPercent=0}
    B=ScalarParameter() {defaultValue=0.5, group=Crystal, parameterName=FresnelIntensity, sortPriority=20}
   )) {name=Fresnel, nodeColor=<rgba 0 0.4609 1 1>}
  NamedRerouteDeclaration(None=@n15.Result) {name=CrystalGrayscale, nodeColor=<rgba 1 0.1172 0 1>}
  NamedRerouteUsage()
  n72 = TextureCoordinate()
  OneMinus(None=MaterialFunctionCall(
    Axis=MaterialFunctionCall!(UVs=@n72, Normal) {materialFunction=/Engine/Functions/Engine_MaterialFunctions01/Texturing/DeriveTangentBasis}.Binormal
    Exponent=@n68
    Normal=PixelNormalWS()
   ) {materialFunction=&AxisAlignedFresnel}.Fresnel)
  SkyAtmosphereLightDirection()
  TextureSampleParameter2D(UVs=@n36) {
   group=Crystal
   parameterName=BacklightNormal
   samplerType=SAMPLERTYPE_Normal
   sortPriority=18
   texture=&T_MF_Crystal3
  }
  TextureSample() {texture=/BRCosmetics/Weapons/FORT_Melee/Pickaxe_CellSeed/Textures/T_CellSeed_Pickaxe_E}
  Transform(None=@n1) {transformSourceType=TRANSFORMSOURCE_Local}
 ]}
